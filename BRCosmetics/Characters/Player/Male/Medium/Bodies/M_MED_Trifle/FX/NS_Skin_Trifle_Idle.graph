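# Idle Niagara effect for the Trifle character skin, crackling blue lightning over the body's skeletal mesh.
# LightningMeshes spawns random bolt variants, some traveling, with flare and glow sprites at bones and offsets from user arrays, gated by PowersActive.
# LightningSprites scatters short-lived sparks and lights across the mesh.
# System timer, flicker and time-of-day values modulate both; hidden head/hand/foot flags set visibility tags, while Glow and debug emitters are disabled.
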NiagaraSystem /BRCosmetics/Characters/Player/Male/Medium/Bodies/M_MED_Trifle/FX/NS_Skin_Trifle_Idle
dependencies {
 usedDataInterfaces [&NiagaraDataInterfaceCurve, &NiagaraDataInterfaceArrayFloat, &NiagaraDataInterfaceColorCurve]
 usedDynamicInputs [&FN_FrontEndFloatModifier, &Boolean_NotOperation]
 usedModules [
  &self:SystemSpawnScript.SystemScriptSource.SystemScriptGraph.NiagaraNodeAssignment_1.SetVariables_#1
  &SystemState
  &self:SystemSpawnScript.SystemScriptSource.SystemScriptGraph.NiagaraNodeAssignment_0.SetVariables_#2
  &self:SystemSpawnScript.SystemScriptSource.SystemScriptGraph.NiagaraNodeAssignment_2.SetVariables_#3
 ]
}
emitters {
 FN_Debug {
  bEnabled=false
  emitterUpdateScript [
   &EmitterState()
   &SpawnBurst_Instantaneous(Spawn Count=1, Use Spawn Probability=false)
  ]
  particleSpawnScript [/Niagara/Modules/Spawn/Initialization/InitializeParticle(
    Lifetime=5
    Module.Write Color=true
    Module.Write Lifetime=true
    Module.Write Mass=true
    Module.Write Material Random=true
    Module.Write Position=false
    Module.Write Scale=false
    Module.Write SpriteRotation=false
    Module.Write SpriteSize=true
    Position=&SimulationPosition()
    Sprite Size=<100 100>
   )]
  particleUpdateScript [
   &ParticleState()
   &FN_Timer()
   /SetVariables(Particles.Debug:NiagaraFloat=$System.FN_Timer.Value) {isSetParameters=true}
   &DynamicMaterialParameters(
    Index 0 Param 1=$Particles.Debug
    Module.Param0WriteEnabled=true
    Module.Param1WriteEnabled=true
    Module.Param2WriteEnabled=true
    Module.Param3WriteEnabled=true
   )
   &CameraOffset(Camera Offset Amount=111)
  ]
  rendererClasses [&NiagaraSpriteRendererProperties]
  renderers [{rendererClass=&NiagaraSpriteRendererProperties}]
 }
 Glow {
  bEnabled=false
  emitterUpdateScript [
   &EmitterState(
    Inactive Response=::NewEnumerator1
    Life Cycle Mode=::NewEnumerator1
    Loop Behavior=::NewEnumerator1
    Loop Duration Mode=::NewEnumerator1
    UseLoopDelay=false
   )
   &SpawnBurst_Instantaneous(Use Spawn Probability=false, UseLoopCountLimit=false)
  ]
  particleSpawnScript [
   /Niagara/Modules/Spawn/Initialization/V2/InitializeParticle(
    Color=<rgba 0.07 0.225 1 1>
    Position Offset=<0 0 25>
    Sprite Size Mode=::NewEnumerator3
    Uniform Sprite Size=66
    UsePositionOffset=true
    Write Material Random=true
   )
   /SetVariables() {isSetParameters=true}
  ]
  particleUpdateScript [
   &ParticleState(Kill Particles When Lifetime Has Elapsed=false)
   /SetVariables(
    Particles.LightColor=&MakeLinearColorFromVectorAndFloat(
     "Vector (RGB)"=&Multiply_VectorByFloat(Vector=&MakeVectorFromLinearColorRGB(Color=$Particles.Color), Float=3)
     "Float (Alpha)"=&Multiply_Float(A=$System.FN_Timer.Value, B=&Lerp_Float(A=0.75, B=1, Alpha=$System.Flicker))
    )
    Particles.LightExponent:NiagaraFloat=&Lerp_Float(A=1.666, B=2.333, Alpha=$System.Flicker)
    Particles.LightRadius=&Lerp_Float(A=77, B=88, Alpha=$System.Flicker)
    Particles.MeshColor:LinearColor=&MakeLinearColorFromVectorAndFloat(
     "Vector (RGB)"=&Multiply_VectorByFloat(Vector=&MakeVectorFromLinearColorRGB(Color=$Particles.Color), Float=11)
     "Float (Alpha)"=&Multiply_Float(A=$System.FN_Timer.Value, B=&Lerp_Float(A=0.4, B=1, Alpha=$System.Flicker))
    )
    Particles.SpriteColor=&MakeLinearColorFromVectorAndFloat(
     "Vector (RGB)"=&Multiply_VectorByFloat(
      Vector=&MakeVectorFromLinearColorRGB(Color=$Particles.Color)
      Float=&FNTimeOfDayScalar(Night Scale=0.3, Morning Scale=0.2, Day Scale=0.2, Evening Scale=0.2)
     )
     "Float (Alpha)"=&Multiply_Float(A=$System.FN_Timer.Value, B=&Lerp_Float(A=0.8, B=1, Alpha=$System.Flicker))
    )
    Particles.SpriteSize=&Multiply_Vector2D_ByFloat(Float=&Ceil_Float(Float=$System.FN_Timer.Value), Vector2D=$Particles.Initial.SpriteSize)
   ) {isSetParameters=true}
  ]
  rendererClasses [&NiagaraLightRendererProperties, &NiagaraSpriteRendererProperties]
  renderers [
   {rendererClass=&NiagaraLightRendererProperties, rendererIndex=0}
   {rendererClass=&NiagaraSpriteRendererProperties, rendererIndex=1}
  ]
 }
 LightningMeshes {
  emitterUpdateScript [
   &EmitterState(
    Inactive Response=::NewEnumerator2
    Life Cycle Mode=::NewEnumerator1
    Loop Duration=&Lerp_Float(A=0.3, B=&RandomRangeFloat(Minimum=0.033, Maximum=0.1), Alpha=$System.FN_Timer.Value)
    Recalculate Duration Each Loop=true
    UseLoopDelay=false
   )
   &SpawnBurst_Instantaneous(
    Spawn Count=&MakeIntFromFloat(Float=&MakeCustomFloatFromBool(Bool=$User.PowersActive, True Float=1, False Float=0))
    Use Spawn Probability=false
    UseLoopCountLimit=false
   )
   /SetVariables(Emitter.DepthFade=35) {isSetParameters=true}
  ]
  particleSpawnScript [
   /SetVariables(
    Particles.MeshIndex=&RandomRangeInt(Minimum=0, Maximum=17)
    Particles.bIsTravelingBolt:NiagaraBool=&SetBoolByFloatComparison(A=&RandomRangeFloat(Minimum=0, Maximum=1), B=0.3)
   ) {isSetParameters=true}
   /SetVariables(Particles.MeshIndex=1) {enabled=false, isSetParameters=true}
   &InitializeParticle(
    Color=<rgba 0 48.0998 222 1>
    Lifetime=&MakeCustomFloatFromBool(Bool=$Particles.bIsTravelingBolt, True Float=0.444, False Float=0.222)
    Mesh Scale Mode=::NewEnumerator3
    Position Offset=&SelectVectorFromArray(Direct Array Index=$Particles.MeshIndex, Vector Selection Array=$User.LightningOffset)
    Sprite Rotation Mode=::NewEnumerator2
    Sprite Size Mode=::NewEnumerator3
    Uniform Sprite Size=&Multiply_Float(
     A=&RandomRangeFloat(Minimum=17, Maximum=33)
     B=&SelectFloatFromArray(
      Direct Array Index=$Particles.MeshIndex
      Float Selection Array=DataInterface() {propertyValues {FloatData [1,1,1,1,0,0,1,1,1,1,1,1,1,1,1,1,1,1]}}
     )
    )
    UsePositionOffset=true
    Write Material Random=false
   )
   &DynamicMaterialParameters(
    Index 0 Param 1=$System.LightningWidth
    Index 0 Param 2=&MakeCustomFloatFromBool(Bool=$Particles.bIsTravelingBolt, True Float=0.75, False Float=3)
    Index 0 Param 4=0.5
    Module.Param0WriteEnabled=true
    Module.Param1WriteEnabled=true
    Module.Param2WriteEnabled=true
    Module.Param3WriteEnabled=true
   )
   /SetVariables(Particles.FlareCameraOffset=1) {isSetParameters=true}
   /SetVariables(Particles.GlowGlowColorInit:LinearColor=<rgba 0 0.264 1 0.5>, Particles.GlowSizeInit:Vector2f=<25 25>) {isSetParameters=true}
  ]
  particleUpdateScript [
   &SkeletalMeshLocation(
    Bone / Socket Index=&SelectIntFromArray(Direct Array Index=$Particles.MeshIndex, Int Selection Array=$User.LightningMeshBoneIndex)
    Bone Sampling Mode=::NewEnumerator4
    MaskSpawnGroup=false
    Sampled Position Offset=&SelectVectorFromArray(Direct Array Index=$Particles.MeshIndex, Vector Selection Array=$User.LightningOffset)
    Skeletal Mesh=$User.SkeletalMesh
   )
   /SetVariables(Particles.MeshPosition:NiagaraPosition=$Particles.Position) {isSetParameters=true}
   &SkeletalMeshLocation(
    Bone / Socket Index=&SelectIntFromArray(Direct Array Index=$Particles.MeshIndex, Int Selection Array=$User.LightningMeshBoneIndex)
    Bone Sampling Mode=::NewEnumerator4
    MaskSpawnGroup=false
    Sampled Position Offset=&SelectVectorFromArray(Direct Array Index=$Particles.MeshIndex, Vector Selection Array=$User.FlareOffset)
    Skeletal Mesh=$User.SkeletalMesh
   ) {module=SkeletalMeshLocation001}
   /SetVariables(Particles.FlarePosition:NiagaraPosition=$Particles.Position) {isSetParameters=true}
   &ParticleState()
   &DynamicMaterialParameters(
    Index 0 Param 1=&MakeCustomFloatFromBool(
     Bool=$Particles.bIsTravelingBolt
     True Float=$System.LightningWidth
     False Float=&FloatFromCurve(FloatCurve=DataInterface() {propertyValues {Curve {keys [<at 0.666 1>,<at 1 0>]}}}, Scale Curve=$System.LightningWidth)
    )
    Index 0 Param 3=&MakeCustomFloatFromBool(
     Bool=$Particles.bIsTravelingBolt
     True Float=$Particles.NormalizedAge
     False Float=&FloatFromCurve(FloatCurve=DataInterface() {propertyValues {Curve {keys [<at 0 0>,<at 0.333 0.5>]}}})
    )
    Index 0 Param 4=&MakeCustomFloatFromBool(
     Bool=$Particles.bIsTravelingBolt
     True Float=0.5
     False Float=&FloatFromCurve(FloatCurve=DataInterface() {propertyValues {Curve {keys [{tangentMode=RCTM_User, time=0, value=0.333}, {tangentMode=RCTM_User, time=1, value=0.666}]}}})
    )
    Module.Param0WriteEnabled=true
    Module.Param1WriteEnabled=false
    Module.Param2WriteEnabled=true
    Module.Param3WriteEnabled=true
   ) {module=DynamicMaterialParameters001}
   &ScaleSpriteSize(Uniform Curve Sprite Scale=DataInterface() {propertyValues {Curve {keys [
        {arriveTangent=3.7119, interpMode=RCIM_Cubic, leaveTangent=3.7119, tangentMode=RCTM_User, time=0.3, value=0.25}
        {interpMode=RCIM_Cubic, tangentMode=RCTM_User, time=0.5, value=1}
       ]}}})
   &ScaleColor(
    Scale RGB=&VectorFromFloat(Value=&Lerp_Float(A=0.133, B=1, Alpha=$System.Flicker))
    ScaleA=false
    ScaleRGB=true
   )
   /SetVariables(Particles.FlareColor=&ColorFromCurve(ColorCurve=DataInterface() {propertyValues {
       AlphaCurve {keys [<at 0 1 RCIM_Cubic>,<at 0.7365 0 RCIM_Cubic>]}
       BlueCurve {keys [<at 0 33 RCIM_Cubic>,<at 0.274 11>,<at 1 1 RCIM_Cubic>]}
       GreenCurve {keys [<at 0 16.9578 RCIM_Cubic>,<at 0.274 6.1613>,<at 1 0.3869 RCIM_Cubic>]}
       RedCurve {keys [<at 0 0 RCIM_Cubic>,<at 0.274 1.8419>,<at 1 0.1451 RCIM_Cubic>]}
      }})) {isSetParameters=true}
   /SetVariables(
    Particles.FlareColor=&MultiplyLinearColor(
     Linear Color=$Particles.FlareColor
     Scale Factor=&MakeFloat4FromVectorAndFloat("Vector (RGB)"=<1 1 1>, "Float (Alpha)"=$System.TimeOfDayBrightnessMultiplier)
    )
    Particles.GlowColor=&MakeLinearColorFromVectorAndFloat(
     "Vector (RGB)"=&MakeVectorFromLinearColorRGB(Color=$Particles.GlowGlowColorInit)
     "Float (Alpha)"=&FloatFromCurve(
      FloatCurve=DataInterface() {propertyValues {Curve {keys [<at 0 0.5>, {interpMode=RCIM_Cubic, tangentMode=RCTM_User, time=0.3, value=1}, {tangentMode=RCTM_User, time=0.75, value=0}]}}}
      Scale Curve=$System.TimeOfDayBrightnessMultiplier
     )
    )
    Particles.GlowSize=&Multiply_Vector2D_ByFloat(
     Float=&FloatFromCurve(FloatCurve=DataInterface() {propertyValues {Curve {keys [{arriveTangent=0.3811, interpMode=RCIM_Cubic, leaveTangent=0.3811, tangentMode=RCTM_User, time=0.3, value=1}, <at 0.5 0>]}}})
     Vector2D=$Particles.GlowSizeInit
    )
   ) {isSetParameters=true}
   &ScaleMeshSize(Scale Factor=&SelectVectorFromArray(
     Direct Array Index=$Particles.MeshIndex
     Vector Selection Array=DataInterface() {propertyValues {FloatData [
        <1 1 1>
        <1 1 1>
        <1 1 1>
        <1 1 1>
        <0.5 0.5 0.5>
        <0.3 0.3 0.3>
        <1 1 1>
        <1 1 1>
        <1 1 1>
        <1 1 1>
        <1 1 1>
        <1 1 1>
        <1 1 1>
        <1 1 1>
        <1 1 1>
        <1 1 1>
        <1 1 1>
        <1 1 1>
       ]}}
    ))
   &ScaleColor(
    Color Value To Scale=$Particles.Color
    Scale RGB=&Multiply_VectorByFloat(Vector=<1 1 1>, Float=$System.TimeOfDayBrightnessMultiplier)
    ScaleA=true
    ScaleRGB=true
   ) {module=ScaleColor001}
   /SetVariables(Particles.VisibilityTag=&self:PartHiddenBoneSelection(
     is Hand Hidden:NiagaraBool=$User.bisHandHidden
     is Head Hidden:NiagaraBool=$User.bisHeadHIdden
     is Foot Hidden:NiagaraBool=$User.bisFootHidden
     Bone Selection:NiagaraInt32=&SelectIntFromArray(Direct Array Index=$Particles.MeshIndex, Int Selection Array=$User.LightningMeshBoneIndex)
    )) {isSetParameters=true}
  ]
  rendererClasses [&NiagaraMeshRendererProperties, &NiagaraSpriteRendererProperties]
  renderers [
   {rendererClass=&NiagaraMeshRendererProperties, rendererIndex=0}
   {rendererClass=&NiagaraSpriteRendererProperties, rendererIndex=1}
   {rendererClass=&NiagaraSpriteRendererProperties, rendererIndex=2}
  ]
 }
 LightningSprites {
  emitterUpdateScript [
   &EmitterState()
   &SpawnRate(
    Spawn Probability=0.666
    SpawnRate=&Multiply_Float(A=$System.FN_Timer.Value, B=11)
    Use Spawn Probability=true
   )
   &SpawnBurst_Instantaneous(Spawn Count=9, Spawn Group=1, Use Spawn Probability=false, UseLoopCountLimit=false)
  ]
  particleSpawnScript [
   &InitializeParticle(
    Color=<rgba 0 4.7666 22 1>
    Lifetime=&MaskFloatBySpawnGroup(
     Masked Float=&RandomRangeFloat(Minimum=0.066, Maximum=0.133)
     Passthrough Float=&RandomRangeFloat(Minimum=0.166, Maximum=0.3)
     Spawn Group Mask=1
    )
    Sprite Rotation Mode=::NewEnumerator2
    Sprite Size Max=<26 13>
    Sprite Size Min=<13 26>
    Sprite Size Mode=::NewEnumerator2
    Sprite UV Mode=::NewEnumerator3
    UsePositionOffset=false
    Write Material Random=true
   ) {module=InitializeParticle001}
   /SetVariables(
    Particles.SpriteSize=&Multiply_Vector2D_ByFloat(
     Float=&MaskFloatBySpawnGroup(Masked Float=&RandomRangeFloat(Minimum=1, Maximum=2), Passthrough Float=1, Spawn Group Mask=1)
     Vector2D=$Particles.SpriteSize
    )
    Particles.SubImageIndex=&UniformRangedFloat(Minimum=0, Maximum=3)
   ) {isSetParameters=true}
   &CameraOffset(Camera Offset Amount=&RandomRangeFloat(Minimum=22, Maximum=33))
   &ShapeLocation(Box Size=<22 11 11>, MaskSpawnGroup=false, Override Seed=false, Shape Primitive=::NewEnumerator2, Surface Only=false, UseOwnerScale=true)
   &Color(Color=$Particles.Color, Scale Alpha=&RandomRangeFloat(Minimum=0.1, Maximum=1))
  ]
  particleUpdateScript [
   &DynamicMaterialParameters(
    Index 0 Param 1=&RandomRangeFloat(Minimum=0, Maximum=1)
    Index 0 Param 2=144
    Index 0 Param 3=&FloatFromCurve(FloatCurve=DataInterface() {propertyValues {Curve {keys [
         {arriveTangent=-19.4071, interpMode=RCIM_Cubic, leaveTangent=-19.4071, tangentMode=RCTM_User, time=0, value=1}
         {interpMode=RCIM_Cubic, tangentMode=RCTM_Break, time=0.1, value=0}
         {arriveTangent=3.3801, interpMode=RCIM_Cubic, leaveTangent=3.3801, tangentMode=RCTM_User, time=1, value=1}
        ]}}})
    Module.Param0WriteEnabled=true
    Module.Param1WriteEnabled=true
    Module.Param2WriteEnabled=true
    Module.Param3WriteEnabled=false
   )
   &ParticleState()
   &ScaleColor(
    Scale Alpha=&FloatFromCurve(
     FloatCurve=DataInterface() {propertyValues {Curve {keys [<at 0.75 1 RCIM_Cubic>, {arriveTangent=-10.3917, interpMode=RCIM_Cubic, leaveTangent=-10.3917, tangentMode=RCTM_User, time=1, value=0}]}}}
     Scale Curve=&Lerp_Float(A=0.333, B=1, Alpha=$System.Flicker)
    )
    ScaleA=true
    ScaleRGB=false
   ) {module=ScaleColor002}
   &ScaleSpriteSize(
    Uniform Curve Scale=&Lerp_Float(A=0.9, B=1, Alpha=$System.Flicker)
    Uniform Curve Sprite Scale=DataInterface() {propertyValues {Curve {keys [{arriveTangent=1.4996, time=0, value=0.9}, <at 0.2 1>, <at 0.7 0.98>, {leaveTangent=-0.2857, time=1, value=1}]}}}
   )
   &SkeletalMeshLocation(Bone Sampling Mode=::NewEnumerator0, MaskSpawnGroup=false, OverrideRandomSeed=false, Skeletal Mesh=$User.SkeletalMesh)
   &ScaleColor(
    Color Value To Scale=$Particles.Color
    Scale RGB=&Multiply_VectorByFloat(Vector=<1 1 1>, Float=$System.TimeOfDayBrightnessMultiplier)
    ScaleA=true
    ScaleRGB=true
   ) {module=ScaleColor001}
   /SetVariables(
    Particles.LightEnabled:NiagaraBool=&SetBoolByIntComparison(
     A=&self:PartHiddenBoneSelection(
      is Hand Hidden:NiagaraBool=$User.bisHandHidden
      is Head Hidden:NiagaraBool=$User.bisHeadHIdden
      is Foot Hidden:NiagaraBool=false
      Bone Selection:NiagaraInt32=$Output.SkeletalMeshLocation.SampledBoneIndex
     )
     B=0
     Comparison Type=::NewEnumerator4
    )
    Particles.VisibilityTag=&self:PartHiddenBoneSelection(
     is Hand Hidden:NiagaraBool=$User.bisHandHidden
     is Head Hidden:NiagaraBool=$User.bisHeadHIdden
     is Foot Hidden:NiagaraBool=$User.bisFootHidden
     Bone Selection:NiagaraInt32=$Output.SkeletalMeshLocation.SampledBoneIndex
    )
   ) {isSetParameters=true}
  ]
  rendererClasses [&NiagaraSpriteRendererProperties, &NiagaraLightRendererProperties]
  renderers [
   {rendererClass=&NiagaraSpriteRendererProperties}
   {rendererClass=&NiagaraLightRendererProperties, rendererIndex=1}
  ]
 }
}
summary {
 systemName=NS_Skin_Trifle_Idle
 emitters [
  {
   emitterName=LightningSprites
   rendererClasses [&NiagaraSpriteRendererProperties, &NiagaraLightRendererProperties]
  }
  {bEnabled=false, emitterName=FN_Debug, rendererClasses [&NiagaraSpriteRendererProperties]}
  {
   bEnabled=false
   emitterName=Glow
   rendererClasses [&NiagaraLightRendererProperties, &NiagaraSpriteRendererProperties]
  }
  {
   emitterName=LightningMeshes
   rendererClasses [&NiagaraMeshRendererProperties, &NiagaraSpriteRendererProperties]
  }
 ]
 userVariables [
  User.bisFootHidden {type=NiagaraBool}
  User.bIsFrontEnd {type=NiagaraBool}
  User.bIsFrontEndPreview {type=NiagaraBool}
  User.bisHandHidden {type=NiagaraBool}
  User.bisHeadHIdden {type=NiagaraBool}
  User.FlareOffset {
   type=NiagaraDataInterfaceArrayFloat3
   description=""
   defaultValue {
    struct=NiagaraExt_VariableValue_DataInterface
    value {
     dataInterface=&self:NiagaraDataInterfaceArrayFloat3_0
     dataInterfaceClass=&NiagaraDataInterfaceArrayFloat3
    }
   }
  }
  User.FlickerSpeed {type=NiagaraFloat, defaultValue=6.666, description=""}
  User.LightningMeshBoneIndex {
   type=NiagaraDataInterfaceArrayInt32
   description=""
   defaultValue {
    struct=NiagaraExt_VariableValue_DataInterface
    value {
     dataInterface=&self:NiagaraDataInterfaceArrayInt32_1
     dataInterfaceClass=&NiagaraDataInterfaceArrayInt32
    }
   }
  }
  User.LightningOffset {
   type=NiagaraDataInterfaceArrayFloat3
   description=""
   defaultValue {
    struct=NiagaraExt_VariableValue_DataInterface
    value {
     dataInterface=&self:NiagaraDataInterfaceArrayFloat3_1
     dataInterfaceClass=&NiagaraDataInterfaceArrayFloat3
    }
   }
  }
  User.LightningSpriteDepthFade {type=NiagaraFloat, defaultValue=44, description=""}
  User.PowersActive {type=NiagaraBool, defaultValue=true}
  User.SkeletalMesh {
   type=NiagaraDataInterfaceSkeletalMesh
   defaultValue {
    struct=NiagaraExt_VariableValue_DataInterface
    value {dataInterface=&self:NiagaraDataInterfaceSkeletalMesh_0}
   }
  }
 ]
}
