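# Physics asset for the male medium Trifle player body cosmetic, set up for secondary motion of its front loincloth.
# Left and right three-bone loincloth chains use capsules linked segment to segment and anchored to the pelvis, with tight twist and asymmetric swing limits.
# A small pelvis capsule and two large thigh capsules complete the body set, with the thighs joined to the pelvis by free angular constraints, likely as colliders for the cloth chains.
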PhysicsAsset /BRCosmetics/Characters/Player/Male/Medium/Bodies/M_MED_Trifle/Meshes/M_MED_Trifle_Physics
bodies {
 dyn_L_front_Loincloth_1 [{shapeType=Capsule, length=3, radius=4, shapeName=dyn_loincloth_01_r_capsule, center=<6 0 0>, rotation=<rot -90 0 0>}]
 dyn_L_front_Loincloth_2 [{shapeType=Capsule, length=3, radius=4, shapeName=dyn_loincloth_01_r_capsule, center=<6 0 0>, rotation=<rot -90 0 0>}]
 dyn_L_front_Loincloth_3 [{shapeType=Capsule, length=3, radius=4, shapeName=dyn_loincloth_01_r_capsule, center=<6 0 0>, rotation=<rot -90 0 0>}]
 dyn_R_front_Loincloth_1 [{shapeType=Capsule, length=3, radius=4, shapeName=dyn_loincloth_01_r_capsule, center=<6 0 0>, rotation=<rot -90 0 0>}]
 dyn_R_front_Loincloth_2 [{shapeType=Capsule, length=3, radius=4, shapeName=dyn_loincloth_01_r_capsule, center=<6 0 0>, rotation=<rot -90 0 0>}]
 dyn_R_front_Loincloth_3 [{shapeType=Capsule, length=3, radius=4, shapeName=dyn_loincloth_01_r_capsule, center=<6 0 0>, rotation=<rot -90 0 0>}]
 pelvis [{shapeType=Capsule, length=1, radius=1, shapeName=pelvis_capsule}]
 thigh_l [{shapeType=Capsule, length=30, radius=8, center=<-19 0 -1.3>, rotation=<rot 90 0 0>}]
 thigh_r [{shapeType=Capsule, length=30, radius=8, shapeName=thigh_twist_01_r_capsule, center=<19 0 1.3>, rotation=<rot 90 0 0>}]
}
constraints [
 {bone1Name=dyn_L_front_Loincloth_2, bone2Name=dyn_L_front_Loincloth_1, swing1LimitDegrees=45, swing2LimitDegrees=15, twistLimitDegrees=5}
 {bone1Name=dyn_L_front_Loincloth_3, bone2Name=dyn_L_front_Loincloth_2, swing1LimitDegrees=45, swing2LimitDegrees=15, twistLimitDegrees=5}
 {bone1Name=dyn_R_front_Loincloth_2, bone2Name=dyn_R_front_Loincloth_1, swing1LimitDegrees=45, swing2LimitDegrees=15, twistLimitDegrees=5}
 {bone1Name=dyn_R_front_Loincloth_3, bone2Name=dyn_R_front_Loincloth_2, swing1LimitDegrees=45, swing2LimitDegrees=15, twistLimitDegrees=5}
 {
  bone1Name=thigh_l
  bone2Name=pelvis
  swing1LimitDegrees=45
  swing1Motion=Free
  swing2LimitDegrees=45
  swing2Motion=Free
  twistLimitDegrees=45
  twistMotion=Free
 }
 {
  bone1Name=thigh_r
  bone2Name=pelvis
  swing1LimitDegrees=45
  swing1Motion=Free
  swing2LimitDegrees=45
  swing2Motion=Free
  twistLimitDegrees=45
  twistMotion=Free
 }
 {bone1Name=dyn_L_front_Loincloth_1, bone2Name=pelvis, swing1LimitDegrees=45, swing2LimitDegrees=15, twistLimitDegrees=5}
 {bone1Name=dyn_R_front_Loincloth_1, bone2Name=pelvis, swing1LimitDegrees=45, swing2LimitDegrees=15, twistLimitDegrees=5}
]
