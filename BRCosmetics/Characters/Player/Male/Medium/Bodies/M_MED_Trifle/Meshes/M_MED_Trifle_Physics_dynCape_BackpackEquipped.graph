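# Physics asset for the M_MED_Trifle male medium body, the dynamic cape variant used when a backpack is equipped.
# Capsules cover pelvis, spine, arms, legs and feet, joined by mostly free angular constraints, alongside bodies for the dyn_ cloth bones.
# The cape hangs from spine_03 through dyn_cape04, then splits into left and right five-bone chains with tight twist limits.
# A three-bone back loincloth chain hangs from the pelvis with twist locked and sideways swing locked at its root.
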PhysicsAsset /BRCosmetics/Characters/Player/Male/Medium/Bodies/M_MED_Trifle/Meshes/M_MED_Trifle_Physics_dynCape_BackpackEquipped
bodies {
 calf_l [{shapeType=Capsule, length=30, radius=11, shapeName=calf_l_capsule, center=<-21.7961 -3.447 -0.0024>, rotation=<rot 90 0 0>}]
 calf_r [{shapeType=Capsule, length=30, radius=11, shapeName=calf_r_capsule, center=<21.7961 3.447 0.0024>, rotation=<rot 90 0 0>}]
 dyn_bk_Loincloth_1 [{shapeType=Capsule, length=20, radius=4.5, shapeName=dyn_bk_Loincloth_1_capsule, center=<2 0 0>}]
 dyn_bk_Loincloth_2 [{shapeType=Capsule, length=20, radius=4.5, shapeName=dyn_bk_Loincloth_2_capsule, center=<2 0 0>}]
 dyn_bk_Loincloth_3 [{shapeType=Capsule, length=20, radius=4.5, shapeName=dyn_bk_Loincloth_3_capsule, center=<2 0 0>}]
 dyn_cape01_l [{shapeType=Capsule, length=15, radius=9, center=<4 -2 0>}]
 dyn_cape01_r [{shapeType=Capsule, length=15, radius=9, center=<4 -2 0>}]
 dyn_cape02_l [{shapeType=Capsule, length=15, radius=9, center=<4 -2 0>}]
 dyn_cape02_r [{shapeType=Capsule, length=15, radius=9, center=<4 -2 0>}]
 dyn_cape03_l [{shapeType=Capsule, length=15, radius=9, center=<4 -2 0>}]
 dyn_cape03_r [{shapeType=Capsule, length=15, radius=9, center=<4 -2 0>}]
 dyn_cape04 [{shapeType=Capsule, length=40, radius=7, center=<4.0879 -2.3 0>}]
 dyn_cape04_l [{shapeType=Capsule, length=20, radius=9, center=<4 -2 0>}]
 dyn_cape04_r [{shapeType=Capsule, length=20, radius=9, center=<4 -2 0>}]
 dyn_cape05_l [{shapeType=Capsule, length=20, radius=9, center=<4 -2 0>}]
 dyn_cape05_r [{shapeType=Capsule, length=20, radius=9, center=<4 -2 0>}]
 foot_l [{shapeType=Capsule, length=15, radius=14.75, shapeName=foot_l_capsule, center=<-4 6 0>, rotation=<rot 0 0 90>}]
 foot_r [{shapeType=Capsule, length=15, radius=14.75, shapeName=foot_l_capsule, center=<4 -6 0>, rotation=<rot 0 0 90>}]
 lowerarm_l [{shapeType=Capsule, length=20, radius=6.5, shapeName=lowerarm_l_capsule, center=<5 0 0>, rotation=<rot 90 0 0>}]
 lowerarm_r [{shapeType=Capsule, length=20, radius=6.5, shapeName=lowerarm_r_capsule, center=<-5 0 0>, rotation=<rot 90 0 0>}]
 pelvis [{shapeType=Capsule, length=8.6562, radius=13.5388, center=<-2.6215 -1.3695 0>}]
 spine_01 [{shapeType=Capsule, length=10.7187, radius=12.4138, center=<-2.6215 -0.7156 0>}]
 spine_03 [{shapeType=Capsule, length=9.1562, radius=10.7575, center=<-1.953 -0.8675 0.0001>}]
 thigh_l [{shapeType=Capsule, length=30, radius=11, shapeName=thigh_l_capsule, center=<-14.9419 -1.1522 -0.0008>, rotation=<rot 90 0 0>}]
 thigh_r [{shapeType=Capsule, length=30, radius=11, shapeName=thigh_r_capsule, center=<14.9419 1.1522 0.0008>, rotation=<rot 90 0 0>}]
 upperarm_l [{shapeType=Capsule, length=20, radius=6.5, shapeName=upperarm_l_capsule, center=<5 0 0>, rotation=<rot 90 0 0>}]
 upperarm_r [{shapeType=Capsule, length=20, radius=6.5, shapeName=upperarm_r_capsule, center=<-5 0 0>, rotation=<rot 90 0 0>}]
}
constraints [
 {bone1Name=thigh_l, bone2Name=pelvis, swing1LimitDegrees=45, swing2LimitDegrees=45, twistLimitDegrees=45}
 {bone1Name=thigh_r, bone2Name=pelvis, swing1LimitDegrees=45, swing2LimitDegrees=45, twistLimitDegrees=45}
 {
  bone1Name=spine_01
  bone2Name=pelvis
  swing1LimitDegrees=45
  swing1Motion=Free
  swing2LimitDegrees=45
  swing2Motion=Free
  twistLimitDegrees=45
  twistMotion=Free
 }
 {
  bone1Name=spine_03
  bone2Name=spine_01
  swing1LimitDegrees=45
  swing1Motion=Free
  swing2LimitDegrees=45
  swing2Motion=Free
  twistLimitDegrees=45
  twistMotion=Free
 }
 {
  bone1Name=calf_l
  bone2Name=thigh_l
  swing1LimitDegrees=45
  swing1Motion=Free
  swing2LimitDegrees=45
  swing2Motion=Free
  twistLimitDegrees=45
  twistMotion=Free
 }
 {
  bone1Name=calf_r
  bone2Name=thigh_r
  swing1LimitDegrees=45
  swing1Motion=Free
  swing2LimitDegrees=45
  swing2Motion=Free
  twistLimitDegrees=45
  twistMotion=Free
 }
 {bone1Name=dyn_cape04, bone2Name=spine_03, swing1LimitDegrees=45, swing2LimitDegrees=45, twistLimitDegrees=5}
 {bone1Name=dyn_cape01_l, bone2Name=dyn_cape04, swing1LimitDegrees=45, swing2LimitDegrees=45, twistLimitDegrees=5}
 {bone1Name=dyn_cape02_l, bone2Name=dyn_cape01_l, swing1LimitDegrees=45, swing2LimitDegrees=45, twistLimitDegrees=5}
 {bone1Name=dyn_cape03_l, bone2Name=dyn_cape02_l, swing1LimitDegrees=45, swing2LimitDegrees=45, twistLimitDegrees=5}
 {bone1Name=dyn_cape04_l, bone2Name=dyn_cape03_l, swing1LimitDegrees=45, swing2LimitDegrees=45, twistLimitDegrees=5}
 {bone1Name=dyn_cape05_l, bone2Name=dyn_cape04_l, swing1LimitDegrees=45, swing2LimitDegrees=45, twistLimitDegrees=5}
 {bone1Name=dyn_cape01_r, bone2Name=dyn_cape04, swing1LimitDegrees=45, swing2LimitDegrees=45, twistLimitDegrees=5}
 {bone1Name=dyn_cape02_r, bone2Name=dyn_cape01_r, swing1LimitDegrees=45, swing2LimitDegrees=45, twistLimitDegrees=5}
 {bone1Name=dyn_cape03_r, bone2Name=dyn_cape02_r, swing1LimitDegrees=45, swing2LimitDegrees=45, twistLimitDegrees=5}
 {bone1Name=dyn_cape04_r, bone2Name=dyn_cape03_r, swing1LimitDegrees=45, swing2LimitDegrees=45, twistLimitDegrees=5}
 {bone1Name=dyn_cape05_r, bone2Name=dyn_cape04_r, swing1LimitDegrees=45, swing2LimitDegrees=45, twistLimitDegrees=5}
 {
  bone1Name=foot_l
  bone2Name=calf_l
  swing1LimitDegrees=45
  swing1Motion=Free
  swing2LimitDegrees=45
  swing2Motion=Free
  twistLimitDegrees=45
  twistMotion=Free
 }
 {
  bone1Name=foot_r
  bone2Name=calf_r
  swing1LimitDegrees=45
  swing1Motion=Free
  swing2LimitDegrees=45
  swing2Motion=Free
  twistLimitDegrees=45
  twistMotion=Free
 }
 {
  bone1Name=dyn_bk_Loincloth_1
  bone2Name=pelvis
  swing1LimitDegrees=45
  swing2LimitDegrees=20
  swing2Motion=Locked
  twistLimitDegrees=45
  twistMotion=Locked
 }
 {
  bone1Name=dyn_bk_Loincloth_2
  bone2Name=dyn_bk_Loincloth_1
  swing1LimitDegrees=45
  swing2LimitDegrees=20
  twistLimitDegrees=45
  twistMotion=Locked
 }
 {
  bone1Name=dyn_bk_Loincloth_3
  bone2Name=dyn_bk_Loincloth_2
  swing1LimitDegrees=45
  swing2LimitDegrees=20
  twistLimitDegrees=45
  twistMotion=Locked
 }
 {
  bone1Name=upperarm_r
  bone2Name=spine_03
  swing1LimitDegrees=45
  swing1Motion=Free
  swing2LimitDegrees=45
  swing2Motion=Free
  twistLimitDegrees=45
  twistMotion=Free
 }
 {
  bone1Name=lowerarm_r
  bone2Name=upperarm_r
  swing1LimitDegrees=45
  swing1Motion=Free
  swing2LimitDegrees=45
  swing2Motion=Free
  twistLimitDegrees=45
  twistMotion=Free
 }
 {
  bone1Name=upperarm_l
  bone2Name=spine_03
  swing1LimitDegrees=45
  swing1Motion=Free
  swing2LimitDegrees=45
  swing2Motion=Free
  twistLimitDegrees=45
  twistMotion=Free
 }
 {
  bone1Name=lowerarm_l
  bone2Name=upperarm_l
  swing1LimitDegrees=45
  swing1Motion=Free
  swing2LimitDegrees=45
  swing2Motion=Free
  twistLimitDegrees=45
  twistMotion=Free
 }
]
