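# Physics asset for the M_MED_Trifle face accessory mesh, NoCape variant, that defines capsule bodies for hair physics.
# Three hair chains hang from neck_02: a three-bone center chain and two-bone left and right chains.
# Side chains lock twist, while the center chain has narrow twist limits and per-bone swing limits.
# Larger spine_05 and clavicle capsules supply torso and shoulder volumes, with neck and clavicles joined to spine_05 by angularly free constraints.
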PhysicsAsset /BRCosmetics/Characters/Player/Male/Medium/Bodies/M_MED_Trifle/Meshes/Parts/M_MED_Trifle_FaceAcc_Physics_NoCape
bodies {
 clavicle_l [{shapeType=Capsule, length=5, radius=7, shapeName=clavicle_r_capsule, center=<8.0235 0.476 0.2922>, rotation=<rot 90 0 0>}]
 clavicle_r [{shapeType=Capsule, length=5, radius=7, shapeName=clavicle_r_capsule, center=<-8.0235 -0.476 -0.2922>, rotation=<rot 90 0 0>}]
 dyn_c_hair1 [{shapeType=Capsule, length=10, radius=4, shapeName=dyn_c_hair1_capsule, center=<2.5 0 0>, rotation=<rot 90 0 0>}]
 dyn_c_hair2 [{shapeType=Capsule, length=10.0906, radius=4, shapeName=dyn_c_hair2_capsule, center=<1 0 0>, rotation=<rot 90 0 0>}]
 dyn_c_hair3 [{shapeType=Capsule, length=11.8719, radius=4, shapeName=dyn_c_hair3_capsule, center=<1 0 0>, rotation=<rot 90 0 0>}]
 dyn_l_hair1 [{shapeType=Capsule, length=2, radius=2.5, shapeName=dyn_l_hair1_capsule, center=<-2 0 0>, rotation=<rot 90 0 0>}]
 dyn_l_hair2 [{shapeType=Capsule, length=5, radius=2.5, shapeName=dyn_l_hair2_capsule, center=<-1 0 0>, rotation=<rot 90 0 0>}]
 dyn_r_hair1 [{shapeType=Capsule, length=2, radius=2, shapeName=dyn_r_hair1_capsule, center=<2 0 0>, rotation=<rot 90 0 0>}]
 dyn_r_hair2 [{shapeType=Capsule, length=5, radius=2, shapeName=dyn_r_hair2_capsule, center=<1 0 0>, rotation=<rot 90 0 0>}]
 neck_02 [{shapeType=Capsule, length=0.5, radius=0.5, shapeName=neck_02_capsule, rotation=<rot 90 0 0>}]
 spine_05 [{shapeType=Capsule, length=36.7281, radius=18.2, shapeName=spine_05_capsule, center=<-1.9647 4.4623 0>}]
}
constraints [
 {bone1Name=dyn_r_hair1, bone2Name=neck_02, swing1LimitDegrees=20, swing2LimitDegrees=45, twistLimitDegrees=45, twistMotion=Locked}
 {bone1Name=dyn_r_hair2, bone2Name=dyn_r_hair1, swing1LimitDegrees=20, swing2LimitDegrees=45, twistLimitDegrees=45, twistMotion=Locked}
 {bone1Name=dyn_l_hair1, bone2Name=neck_02, swing1LimitDegrees=20, swing2LimitDegrees=45, twistLimitDegrees=45, twistMotion=Locked}
 {bone1Name=dyn_l_hair2, bone2Name=dyn_l_hair1, swing1LimitDegrees=20, swing2LimitDegrees=45, twistLimitDegrees=45, twistMotion=Locked}
 {bone1Name=dyn_c_hair1, bone2Name=neck_02, swing1LimitDegrees=50, swing2LimitDegrees=5, twistLimitDegrees=5}
 {bone1Name=dyn_c_hair2, bone2Name=dyn_c_hair1, swing1LimitDegrees=60, swing2LimitDegrees=60, twistLimitDegrees=5}
 {bone1Name=dyn_c_hair3, bone2Name=dyn_c_hair2, swing1LimitDegrees=60, swing2LimitDegrees=10, twistLimitDegrees=10}
 {
  bone1Name=neck_02
  bone2Name=spine_05
  swing1LimitDegrees=45
  swing1Motion=Free
  swing2LimitDegrees=45
  swing2Motion=Free
  twistLimitDegrees=45
  twistMotion=Free
 }
 {
  bone1Name=clavicle_l
  bone2Name=spine_05
  swing1LimitDegrees=45
  swing1Motion=Free
  swing2LimitDegrees=45
  swing2Motion=Free
  twistLimitDegrees=45
  twistMotion=Free
 }
 {
  bone1Name=clavicle_r
  bone2Name=spine_05
  swing1LimitDegrees=45
  swing1Motion=Free
  swing2LimitDegrees=45
  swing2Motion=Free
  twistLimitDegrees=45
  twistMotion=Free
 }
]
