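# Animation Blueprint for the face accessory part of the Trifle male medium body, in its no-cape variant.
# It copies the attached parent's pose and curves, runs a Control Rig, then simulates RigidBody physics in neck_01 space with the NoCape physics asset.
# The event graph derives neckDelta from head yaw and capeOn from the RestrictCape tag.
# The neckDelta-weighted helmet bone chain does not reach Output Pose, and nothing reads capeOn.
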AnimBlueprint /BRCosmetics/Characters/Player/Male/Medium/Bodies/M_MED_Trifle/Meshes/Parts/M_MED_Trifle_FaceAcc_AnimBP_NoCape
target_skeleton=&/M_MED_Trifle_FaceAcc_Skeleton
graphs [
 AnimGraph {nodes [
   n1 = Get neckDelta() -> (neckDelta:double)
   n2 = "Transform (Modify) Bone - Bone: dyn_helmetRotate"(Alpha=@n1, Translation="0,0,0", Rotation="0.000,0.000,18", Scale="1,1,1") {
    rotationMode=BMM_Additive
    rotationSpace=BCS_BoneSpace
    scaleSpace=BCS_BoneSpace
    alphaBoolBlend {blendOption=Linear}
    boneToModify {boneName=dyn_helmetRotate}
    rotation {roll=18}
   }
   n3 = "Transform (Modify) Bone - Bone: dyn_helmetScaleUpper"(ComponentPose=@n2, Alpha=@n1, Translation="0,0,0", Rotation="0,0,0", Scale="1,1,0.57") {
    rotationSpace=BCS_BoneSpace
    scaleMode=BMM_Additive
    scaleSpace=BCS_BoneSpace
    alphaBoolBlend {blendOption=Linear}
    boneToModify {boneName=dyn_helmetScaleUpper}
    scale {z=0.57}
   }
   "Transform (Modify) Bone - Bone: dyn_helmetScaleLower"(ComponentPose=@n3, Alpha=@n1, Translation="0,0,0", Rotation="0,0,0", Scale="1,0.5,1") {
    rotationSpace=BCS_BoneSpace
    scaleMode=BMM_Additive
    scaleSpace=BCS_BoneSpace
    alphaBoolBlend {blendOption=Linear}
    boneToModify {boneName=dyn_helmetScaleLower}
    scale {y=0.5}
   }
   n4 = Control Rig(Source=Copy Pose From Mesh() {bCopyCurves=true, bUseAttachedParent=true}) {
    alphaBoolBlend {blendOption=Linear}
    controlRigAssetReference {blueprintRigClass=&/M_MED_Trifle_FaceAcc_CtrlRig.M_MED_Trifle_FaceAcc_CtrlRig_C}
   }
   n5 = RigidBody(ComponentPose=Local To Component(LocalPose=@n4)) {
    bClampLinearTranslationLimitToRefPose=true
    bOverrideWorldGravity=true
    overridePhysicsAsset=&/M_MED_Trifle_FaceAcc_Physics_NoCape
    simulationSpace=BaseBoneSpace
    alphaBoolBlend {blendOption=Linear}
    baseBoneRef {boneName=neck_01}
    componentAppliedLinearAccClamp=<3000 3000 3000>
    componentLinearAccScale=<0.4 0.4 0.5>
    componentLinearVelScale=<0.4 0.4 0.5>
    overrideWorldGravity {z=-980}
   }
   Output Pose(Result=Component To Local(ComponentPose=@n5)) {layerGroup=None}
  ]}
 EventGraph {nodes [
   Event BlueprintUpdateAnimation()
   n1 = BreakTransform(InTransform=GetDeltaTransformFromRefPose(self=GetOwnerSkeletalMeshComponent(), BoneName=head, BaseName=neck_02))
   > Set neckDelta(neckDelta:double=MapRangeClamped(Value=BreakRotator(InRot=@n1.Rotation).Yaw, InRangeA=-1, InRangeB=-26)) -> (Output_Get:double)
   n2 = Branch(Condition=CharacterItemDefinitionHasAnyMetaTag(Tags=Make Array("[0]"=(TagName="Cosmetics.Character.RestrictCape"))))
   > Set capeOn(capeOn:bool=true) -> (Output_Get:bool)
   Set capeOn(execute=@n2.else, capeOn=false)
   Event OnCharacterCustomizationLoaded()
  ]}
]
variables [
 Aim {type=number}
 Anim_Inst {type=AnimInstance, default=null}
 Body_Mesh {type=SkeletalMeshComponent}
 capeOn {type=boolean}
 neckDelta {type=number}
 Skel_Mesh {type=SkeletalMeshComponent}
 Target {type=CustomCharacterPartAnimInstance, default=null}
]
