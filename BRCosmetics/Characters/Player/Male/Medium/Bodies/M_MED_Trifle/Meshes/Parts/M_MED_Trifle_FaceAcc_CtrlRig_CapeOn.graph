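# Control Rig for the M_MED_Trifle face accessory, CapeOn variant, that tilts the helmet in response to head orientation.
# A spherical pose reader measures the head bone relative to the pelvis against an offset target region with falloff.
# Its output weights a Modify Transforms step that applies a small fixed X rotation to dyn_helmetRotate, blending the corrective in as the head enters that region.
# Debug drawing is left enabled on the pose reader.
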ControlRigBlueprint /BRCosmetics/Characters/Player/Male/Medium/Bodies/M_MED_Trifle/Meshes/Parts/M_MED_Trifle_FaceAcc_CtrlRig_CapeOn
graphs [
 RigVMFunctionLibrary {links [], nodes []}
 RigVMModel {nodes [
   RigUnit_BeginExecution = &RigUnit_BeginExecution()
   n1 = &RigUnit_SphericalPoseReader(
    ActiveRegionScaleFactors=(PositiveWidth=0.5,NegativeWidth=0.5,PositiveHeight=0.5,NegativeHeight=0.5)
    ActiveRegionSize=0.2
    Debug=(bDrawDebug=True,bDrawLocalAxes=false,DebugScale=25,DebugSegments=20,DebugThickness=0.25)
    DriverAxis=(X=1,Y=0,Z=0)
    DriverItem=(Type=Bone,Name="head")
    FalloffRegionScaleFactors=(PositiveWidth=1,NegativeWidth=1,PositiveHeight=1,NegativeHeight=1)
    FalloffSize=0.25
    OptionalParentItem=(Type=Bone,Name="pelvis")
    RotationOffset=(X=-49,Y=-80,Z=0)
    ExecutePin=@RigUnit_BeginExecution.ExecutePin
   )
   > &RigUnit_ModifyTransforms(
    ItemToModify=((Item=(Type=Bone,Name="dyn_helmetRotate"),Transform=(Rotation=(X=0,Y=0,Z=0,W=1),Translation=(X=0,Y=0,Z=0),Scale3D=(X=1,Y=1,Z=1))))
    ItemToModify.0.Transform.Rotation=&RigVMFunction_MathQuaternionFromEuler(Euler=(X=-13,Y=0,Z=0)).Result
    Weight=@n1.OutputParam
   )
  ]}
]
hierarchy [
 root {loc=<0 0 0>, rot=<0 0 0 1>}
 pelvis {parent=root, loc=<0 -0.7071 93.1191>, rot=<0.0127 0.707 -0.0127 -0.707>}
 spine_01 {parent=pelvis, loc=<6.7494 -0.6388 0>, rot=<0 0 0.034 -0.9994>}
 spine_02 {parent=spine_01, loc=<7.2518 -0.035 0>, rot=<0 0 0.0238 -0.9997>}
 spine_03 {parent=spine_02, loc=<6.0104 0 0>, rot=<0 0 0.0149 -0.9999>}
 spine_04 {parent=spine_03, loc=<8.1024 0 0>, rot=<0 0 -0.0132 -0.9999>}
 spine_05 {parent=spine_04, loc=<10.0146 0 0>, rot=<0 0 -0.0015 -1>}
 neck_01 {parent=spine_05, loc=<11.3138 0.1529 0>, rot=<0 0 -0.2003 -0.9797>}
 neck_02 {parent=neck_01, loc=<4.5627 0 0>, rot=<0 0 0.0504 -0.9987>}
 head {parent=neck_02, loc=<5.3914 0 0>, rot=<0 0 0.115 -0.9934>}
 dyn_r_hair1 {parent=head, loc=<0.2534 -2.1501 5.4626>, rot=<0.1367 -0.9482 -0.0386 0.2843>}
 dyn_r_hair2 {parent=dyn_r_hair1, loc=<3.3048 0 0>, rot=<0.0082 -0.087 -0.0249 -0.9959>}
 dyn_l_hair1 {parent=head, loc=<0.2536 -2.1501 -5.4626>, rot=<0.9482 0.1367 0.2843 0.0386>}
 dyn_l_hair2 {parent=dyn_l_hair1, loc=<-3.3051 -0.0001 -0.0002>, rot=<0.0082 -0.087 -0.0249 -0.9959>}
 dyn_c_hair1 {parent=head, loc=<1.132 -6.3015 0>, rot=<0.3894 -0.9211 0 0>}
 dyn_c_hair2 {parent=dyn_c_hair1, loc=<5.2643 0 0>, rot=<0 0 0.0913 -0.9958>}
 dyn_c_hair3 {parent=dyn_c_hair2, loc=<4.7531 0 0>, rot=<0 0 0.1041 -0.9946>}
 faceAttach {parent=head, loc=<0.6201 0.9296 0>, rot=<0.0029 -0.7071 -0.0029 -0.7071>}
 C_jaw {parent=faceAttach, loc=<-0.0085 1.3733 -0.9769>, rot=<0 0 0 1>}
 C_lip_lower_mid {parent=C_jaw, loc=<0.0064 7.2119 -1.2288>, rot=<0 -0.0042 -0.0009 -1>}
 L_lip_lower_outer {parent=C_jaw, loc=<1.3322 6.4567 -1.1014>, rot=<0.0215 0.0908 0.3401 -0.9357>}
 R_lip_lower_outer {parent=C_jaw, loc=<-1.3522 6.5214 -1.0937>, rot=<0.0223 -0.0952 -0.3433 -0.9341>}
 C_lip_upper_mid {parent=faceAttach, loc=<0.0026 8.7124 -2.008>, rot=<0.2225 0.0008 -0.0003 -0.9749>}
 C_nose_bridge {parent=faceAttach, loc=<0.0001 10.091 1.5977>, rot=<-0.2218 0.0129 0.0036 -0.975>}
 L_brow_outer {parent=faceAttach, loc=<4.8154 6.488 5.1738>, rot=<0.155 0.1913 0.3921 -0.8864>}
 L_cheek_inner {parent=faceAttach, loc=<3.6247 6.6347 0.1932>, rot=<-0.0762 -0.2838 0.3163 -0.902>}
 L_lip_corner {parent=faceAttach, loc=<2.2586 7.2918 -2.1467>, rot=<0.0027 0.0082 -0.0146 -0.9999>}
 L_lip_upper_outer {parent=faceAttach, loc=<1.3334 7.936 -2.0898>, rot=<0.0286 -0.0933 0.2703 -0.9578>}
 R_brow_outer {parent=faceAttach, loc=<-4.8156 6.4875 5.1737>, rot=<0.0603 -0.072 -0.4266 -0.8996>}
 R_cheek_inner {parent=faceAttach, loc=<-3.6247 6.6348 0.1932>, rot=<-0.0782 0.2886 -0.3156 -0.9005>}
 R_eye {parent=faceAttach, loc=<-2.7053 6.3162 3.7636>, rot=<0 0 0.9998 -0.0187>}
 R_eye_lid_lower_mid {parent=R_eye, loc=<0.0223 -0.0103 0.0218>, rot=<0.0329 -0.0005 0.0167 -0.9993>}
 R_lip_corner {parent=faceAttach, loc=<-2.2175 7.3129 -2.1731>, rot=<0.0029 -0.008 0.0146 -0.9999>}
 R_lip_upper_outer {parent=faceAttach, loc=<-1.3337 7.9355 -2.0902>, rot=<0.0278 0.0956 -0.2712 -0.9574>}
 dyn_helmetRotate {parent=head, loc=<6.0263 0.6787 0>, rot=<0.0029 -0.7071 -0.0029 -0.7071>}
 dyn_helmetScaleUpper {parent=dyn_helmetRotate, loc=<0 -8.6257 0.0355>, rot=<0 0 0 1>}
 dyn_helmetScaleLower {parent=dyn_helmetScaleUpper, loc=<0 0.4335 -5.4821>, rot=<0 0 0 1>}
]
